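; Parent Blueprint for Asteria kapok tree branches that plays a looping leaf gameplay cue on Fort pawns standing on it.
; It registers with the basing change subsystem on BeginPlay and unregisters on EndPlay.
; When a pawn bases on the branch, it adds the cue and binds to that pawn's base-changed delegate.
; When the pawn's base changes to anything else, it removes the cue and clears all bindings on that delegate.
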(event EventBeginPlay
  (BasingChangeSubsystem|RegisterActorforBaseChangeNotify (GameInstanceSubsystems|GetBasingChangeSubsystem) self (delegate Custom|OnPawnBasedOnMe)))

(event Custom|OnPawnBasedOnMe
  (params (BasedPawn :type Fort%20Pawn%20Object%20Reference))
  (GameplayCue|AddGameplayCueOnActor(Looping) BasedPawn (Variables|Default|GetGC_PlayerLeafLoopingTag) (Ability|GameplayCue|MakeGameplayCueParameters 0.0))
  (EventDispatchers|BindEventtoOnBaseChangedDelegate BasedPawn (delegate Custom|OnPawnBaseChanged)))

(event Custom|OnPawnBaseChanged
  (params (Pawn :type Fort%20Pawn%20Object%20Reference) (NewBase :type Actor%20Object%20Reference))
  (if (!= NewBase self)
    (GameplayCue|RemoveGameplayCueOnActor(Looping) Pawn (Variables|Default|GetGC_PlayerLeafLoopingTag) (Ability|GameplayCue|MakeGameplayCueParameters 0.0))
    (EventDispatchers|UnbindallEventsfromOnBaseChangedDelegate Pawn)))

(event EventEndPlay
  (params (EndPlayReason :type EEndPlayReason%20Enum :default "Destroyed"))
  (BasingChangeSubsystem|UnregisterActorfromBaseChangeNotify (GameInstanceSubsystems|GetBasingChangeSubsystem) self))

(fn ConstructionScript
  (|CallParentFunction))
