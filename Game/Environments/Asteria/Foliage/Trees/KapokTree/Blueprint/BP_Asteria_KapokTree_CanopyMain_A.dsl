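; Kapok tree main canopy Blueprint that gives pawns a looping leaf gameplay cue while they overlap its box collision.
; On begin overlap it runs a Pawn.Athena tag check, casts the actor to FortPawn, adds the cue and records the pawn in a CanopiedPlayers array.
; On end overlap it removes the cue and array entry only for actors recorded in that array.
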(event EventBeginPlay)

(event Collision|EventActorBeginOverlap
  (params (OtherActor :type Actor%20Object%20Reference)))

(event EventTick
  (params (DeltaSeconds :type Float%20%28single-precision%29)))

(event OnComponentBeginOverlap(BoxCollision)
  (params (OverlappedComponent :type Primitive%20Component%20Object%20Reference) (OtherActor :type Actor%20Object%20Reference) (OtherComp :type Primitive%20Component%20Object%20Reference) (OtherBodyIndex :type Integer) (bFromSweep :type Boolean) (SweepResult :type Hit%20Result%20Structure%20%28by%20ref%29))
  (Class|AthenaQuestFunctionLibrary|AthenaCheckActorHasTags OtherActor "(GameplayTags=((TagName=\"Pawn.Athena\")))")
  (bind _asfort_pawn (Utilities|Casting|CastToFortPawn OtherActor))
  (GameplayCue|AddGameplayCueOnActor(Looping) _asfort_pawn (Variables|Default|GetPlayerLeafCue) (Ability|GameplayCue|MakeGameplayCueParameters 0.0))
  (Utilities|Array|AddUnique (Variables|Default|GetCanopiedPlayers) _asfort_pawn))

(event OnComponentEndOverlap(BoxCollision)
  (params (OverlappedComponent :type Primitive%20Component%20Object%20Reference) (OtherActor :type Actor%20Object%20Reference) (OtherComp :type Primitive%20Component%20Object%20Reference) (OtherBodyIndex :type Integer))
  (if (Utilities|Array|ContainsItem (Variables|Default|GetCanopiedPlayers) OtherActor)
    (Utilities|Array|RemoveItem (Variables|Default|GetCanopiedPlayers) OtherActor)
    (GameplayCue|RemoveGameplayCueOnActor(Looping) OtherActor (Variables|Default|GetPlayerLeafCue) (Ability|GameplayCue|MakeGameplayCueParameters 0.0))))

(fn ConstructionScript
  (|CallParentFunction))
