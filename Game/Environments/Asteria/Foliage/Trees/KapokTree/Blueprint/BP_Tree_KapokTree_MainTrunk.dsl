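; Kapok tree main trunk Blueprint that optionally scatters small plant instances over itself during construction.
; A random stream seeded from Seed plus actor X/Y location fires randomized inward line traces from around the trunk, aiming to keep hits on itself or landscape.
; An instanced static mesh component receives instances aligned to hit normals, plus instances at mesh sockets named with "Plant".
; BeginPlay disables World Position Offset evaluation on the trunk mesh.
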(event EventBeginPlay
  (Rendering|LOD|SetEvaluateWorldPositionOffset (Variables|Debug|GetStaticMeshComponent) false :!class /Script/Engine.StaticMeshComponent))

(fn SpawnSmall_Plants
  (bind _stream (Variables|Default|GetStream))
  (if (Variables|Default|GetSpawnSmallPlants)
    (Math|Random|SetRandomStreamSeed (Variables|Default|GetStream) (Math|Conversions|ToInteger(Integer64) (+ (Variables|Default|GetSeed) (Math|Float|TruncatetoInteger64 (+ (.x (Transformation|GetActorLocation)) (.y (Transformation|GetActorLocation)))))))
    (bind _returnvalue (AddComponent|NODE_AddInstancedStaticMeshComponent-0 :RelativeTransform (Math|Transform|MakeTransform "0, 0, 0")))
    (LocalVariables|SetISM _returnvalue)
    (Utilities|FlowControl|Sequence
      (:then_0
        (for _ (range (+ (Math|Random|RandomIntegerinRangefromStream _stream 10 (LocalVariables|GetMaxTraces)) 1))
          (bind (_outhit _returnvalue_2) (Collision|LineTraceByChannel :Start (Utilities|Operators|Add (* (Math|Vector|RotateVectorAroundAxis "1.000000,0.000000,0.000000" (Math|Random|RandomFloatinRangefromStream _stream 0.0 3.6e+02) "0.000000,0.000000,1.000000") 8e+02) (Transformation|GetActorLocation) (Math|Vector|MakeVector 0.0 0.0 (Math|Random|RandomFloatinRangefromStream (Variables|Default|GetStream) 1e+02 1.7e+03))) :End (+ (Utilities|Operators|Add (* (Math|Vector|RotateVectorAroundAxis "1.000000,0.000000,0.000000" (Math|Random|RandomFloatinRangefromStream _stream 0.0 3.6e+02) "0.000000,0.000000,1.000000") 8e+02) (Transformation|GetActorLocation) (Math|Vector|MakeVector 0.0 0.0 (Math|Random|RandomFloatinRangefromStream (Variables|Default|GetStream) 1e+02 1.7e+03))) (* (Math|Random|RandomUnitVectorinConeinDegreesfromStream (Variables|Default|GetStream) (Math|Vector|NegateVector (Math|Vector|RotateVectorAroundAxis "1.000000,0.000000,0.000000" (Math|Random|RandomFloatinRangefromStream _stream 0.0 3.6e+02) "0.000000,0.000000,1.000000")) 2e+01) 1e+03)) :TraceChannel "TraceTypeQuery10" :bTraceComplex true :DrawDebugType "ForDuration" :bIgnoreSelf false :TraceColor "(R=1.000000,G=0.000000,B=0.000000,A=1.000000)" :TraceHitColor "(R=0.000000,G=1.000000,B=0.000000,A=1.000000)" :DrawTime 3.0))
          (if _returnvalue_2
            (Utilities|FlowControl|Sequence
              (:then_0
                (LocalVariables|SetValidHit (== (.HitActor (Collision|BreakHitResult _outhit)) self)))
              (:then_1
                (Utilities|Casting|CastToLandscapeProxy (.HitActor (Collision|BreakHitResult _outhit)))
                (LocalVariables|SetValidHit true))
              (:then_2
                (if (and (LocalVariables|GetValidHit) _returnvalue_2)
                  (Components|InstancedStaticMesh|AddInstance (LocalVariables|GetISM) (Math|Transform|MakeTransform (+ (.Location (Collision|BreakHitResult _outhit)) (* (.Normal (Collision|BreakHitResult _outhit)) 3.0)) (Math|Rotator|MakeRotfromZ (.Normal (Collision|BreakHitResult _outhit)))) true)))))))
      (:then_1
        (for _array_element (Transformation|GetAllSocketNames (Variables|Debug|GetStaticMeshComponent))
          (if (Utilities|String|Contains (Utilities|String|ToString(Name) _array_element) "Plant")
            (Components|InstancedStaticMesh|AddInstance (LocalVariables|GetISM) (Transformation|GetSocketTransform (Variables|Debug|GetStaticMeshComponent) _array_element "RTS_World") true)))))))

(fn ConstructionScript
  (Default|SpawnSmallPlants))
