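# Master bark material for the Asteria Kapok tree, Nanite-enabled and built on material attributes.
# It layers base bark textures with a tiled second-UV detail set, using contrast-adjusted masks, and adds moss and vine tints, with vertex colors declared for ground falloff, moss and vines.
# Quality and shading-path switches give mobile and lower tiers simpler shading; High keeps only the blended detail normal.
# The result feeds the shared MF_JungleMoss_Trees function.
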Material /Game/Environments/Asteria/Foliage/Trees/KapokTree/Materials/M_Asteria_KapokTree_master
graph {
 properties {bUseMaterialAttributes=true, bUsedWithNanite=true, pixelDepthOffsetMode=PDOM_Legacy, usedShadingModels=Default Lit}
 nodes [
  n1 = NamedRerouteUsage()
  n2 = NamedRerouteUsage()
  n3 = ScalarParameter() {defaultValue=0.25, parameterName=Moss Tint Intensity}
  n4 = LinearInterpolate(
   A=@n1
   B=Multiply(A=@n1, B=VectorParameter() {group=Base Material, parameterName=Moss Tint, defaultValue=<rgba 0.0929 0.2094 0.0076 1>}.RGB)
   Alpha=Saturate(None=Multiply(A=@n2, B=@n3) {constB=0.25})
  )
  n5 = NamedRerouteUsage()
  n6 = MaterialFunctionCall(In=Power(Base=NamedRerouteUsage(), Exp=Constant() {r=1})) {materialFunction=&CheapContrast}
  n7 = NamedRerouteUsage()
  n8 = Multiply(
   A=Subtract(A=Subtract(A=@n6.Result, B=@n7), B=NamedRerouteUsage())
   B=ScalarParameter() {defaultValue=1, group=Detail, parameterName=Detail Mask}
  ) {constB=0.8}
  n9 = Saturate(None=NamedRerouteUsage())
  n10 = Saturate(None=Subtract(A=Subtract(A=@n8, B=@n9), B=NamedRerouteUsage()))
  n11 = OneMinus(None=NamedRerouteUsage())
  n12 = NamedRerouteUsage()
  n13 = ScalarParameter() {defaultValue=5, group=Vines, parameterName=Vines Falloff Exponent}
  n14 = NamedRerouteUsage()
  n15 = Multiply(
   A=Power(Base=Multiply(A=@n12, B=@n13) {constB=5}, Exp=ScalarParameter() {defaultValue=0.25, parameterName=Vines Falloff Intensity}) {constExponent=0.25}
   B=@n14
  )
  n16 = LinearInterpolate(
   A=LinearInterpolate(
    A=LinearInterpolate(A=@n4, B=@n5, Alpha=@n10)
    B=Constant3Vector() {constant=<rgba 0.012 0.0177 0.0043 1>}
    Alpha=Power(Base=@n11) {constExponent=8}
   )
   B=VectorParameter() {group=Vines, parameterName=Vines Base Color, defaultValue=<rgba 0.0719 0.0382 0.0056 1>}.RGB
   Alpha=Saturate(None=MaterialFunctionCall(In=@n15) {materialFunction=&CheapContrast}.Result)
  )
  n17 = Multiply(A=Power(Base=Noise!(World Position, FilterWidth) {scale=0.01}) {constExponent=12}, B=@n14)
  n18 = Power(Base=DotProduct(A=NamedRerouteUsage(), B=NamedRerouteUsage())) {constExponent=5}
  n19 = LinearInterpolate(
   A=@n16
   B=Multiply(A=@n16, B=VectorParameter() {group=Vines, parameterName=Vine TInt, defaultValue=<rgba 0.3903 0.7292 0.262 1>}.RGB)
   Alpha=Saturate(None=Multiply(A=@n17, B=@n18))
  )
  n20 = NamedRerouteUsage()
  n21 = NamedRerouteUsage()
  n22 = ScalarParameter() {defaultValue=5, parameterName=MobileMoss}
  n23 = LinearInterpolate(
   A=@n20
   B=Multiply(A=VectorParameter() {parameterName=LOW MossColor, defaultValue=<rgba 0.2836 0.4792 0 1>}.RGB, B=@n20)
   Alpha=Saturate(None=Multiply(A=@n21, B=@n22))
  )
  n24 = NamedRerouteUsage()
  n25 = Multiply(A=NamedRerouteUsage()) {constB=0.5}
  n26 = Add(A=Multiply(A=LinearInterpolate(A=@n24, B=@n25, Alpha=@n6.Result), B=NamedRerouteUsage()), B=NamedRerouteUsage())
  n27 = NamedRerouteUsage()
  n28 = NamedRerouteUsage()
  n29 = NamedRerouteUsage()
  n30 = Saturate(None=Add(A=LinearInterpolate(A=@n27, B=@n28, Alpha=@n29) {constB=0.9}, B=@n10))
  n31 = NamedRerouteUsage()
  n32 = NamedRerouteUsage()
  n33 = MaterialFunctionCall(BaseNormal=@n31, AdditionalNormal=LinearInterpolate(A=NamedRerouteUsage(), B=@n31, Alpha=Multiply(A=@n6.Result, B=@n32))) {materialFunction=&BlendAngleCorrectedNormals}
  n34 = SetMaterialAttributes!(
   MaterialAttributes
   Base Color=ShadingPathSwitch(Default=QualitySwitch(Default=@n19, Low=@n23, High=@n23, Medium=@n23), Mobile=@n23) {desc=Migrated from FeatureLevelSwitch}
   Specular=ShadingPathSwitch(Default=QualitySwitch(Default=@n26, Low=@n24, High=@n24, Medium=@n24), Mobile=@n24) {desc=Migrated from FeatureLevelSwitch}
   Roughness=ShadingPathSwitch(Default=QualitySwitch(Default=@n30, Low=@n27, High=@n27, Medium=@n27), Mobile=@n27) {desc=Migrated from FeatureLevelSwitch}
   Normal=ShadingPathSwitch(Default=QualitySwitch(Default=@n33.Result, Low=@n31, High=@n33.Result, Medium=@n31), Mobile=@n31) {desc=Migrated from FeatureLevelSwitch}
  ) {attributeSetTypes [
    #1
    #2
    #3
    #4
   ]}
  MaterialFunctionCall!(In=@n34) {materialFunction=/Game/Environments/Asteria/Foliage/Trees/MasterMaterials/MaterialFunctions/MF_JungleMoss_Trees}
  NamedRerouteDeclaration(None=Constant3Vector() {constant=<rgba 0 0 1 0>}) {name=Flat Normal, nodeColor=<rgba 0.3384 0.267 1 1>}
  n35 = TextureCoordinate() {coordinateIndex=1}
  n36 = AppendVector(
   A=Multiply(A=ComponentMask(None=@n35) {r=true}, B=ScalarParameter() {defaultValue=1, group=Detail, parameterName=Detail Tiling X})
   B=Multiply(A=ComponentMask(None=@n35) {g=true}, B=ScalarParameter() {defaultValue=1, group=Detail, parameterName=Detail Tiling Y})
  )
  n37 = TextureSampleParameter2D(UVs=@n36) {
   group=Detail
   parameterName=Detail Base Color Tex
   sortPriority=1
   texture=/Game/Environments/Asteria/Foliage/Trees/KapokTree/Textures/KapokBark_detail_diffuse
  }
  NamedRerouteDeclaration(None=Multiply(A=@n37.RGB, B=VectorParameter() {group=Detail, parameterName=Detail tint, defaultValue=<rgba 1 0.0292 0.9745 1>}.RGB)) {name=Detail Tint, nodeColor=<rgba 0.4531 1 0 1>}
  n38 = TextureCoordinate()
  n39 = TextureSampleParameter2D(UVs=@n38) {
   group=Base Material
   parameterName=Base Material BaseColor Tex
   sortPriority=1
   texture=/Game/Environments/Asteria/Foliage/Trees/KapokTree/Textures/KapokBark_diffuse
  }
  n40 = TextureSampleParameter2D(UVs=@n38) {
   group=Base Material
   parameterName=Base Material SMR Tex
   samplerType=SAMPLERTYPE_Masks
   sortPriority=3
   texture=/Game/Environments/Asteria/Foliage/Trees/KapokTree/Textures/KapokBark_SMR
  }
  n41 = ScalarParameter() {defaultValue=0.9, group=Base Material, parameterName=Base Roughness}
  n42 = TextureSampleParameter2D(UVs=@n38) {
   group=Base Material
   parameterName=Base Material Normal Tex
   samplerType=SAMPLERTYPE_Normal
   sortPriority=2
   texture=/Game/Environments/Asteria/Foliage/Trees/KapokTree/Textures/KapokBark_normal
  }
  n43 = SetMaterialAttributes!(
   MaterialAttributes
   Base Color=Multiply(A=@n39.RGB) {constB=1.2}
   Metallic=Constant()
   Specular=@n40.R
   Roughness=LinearInterpolate(A=@n40.B, Alpha=@n41)
   Normal=LinearInterpolate(A=@n42.RGB, B=NamedRerouteUsage())
  ) {attributeSetTypes [
    #1
    #5
    #2
    #3
    #4
   ]}
  n44 = BreakMaterialAttributes!(Attr=@n43)
  NamedRerouteDeclaration(None=@n44.BaseColor) {name=Base BC, nodeColor=<rgba 0 0.5078 1 1>}
  NamedRerouteDeclaration(None=@n44.Specular) {name=Base Specular, nodeColor=<rgba 0.9978 1 0.946 1>}
  NamedRerouteDeclaration(None=@n44.Roughness) {name=Base Roughness, nodeColor=<rgba 0.176 0.176 0.176 1>}
  n45 = TextureSampleParameter2D(UVs=@n38) {
   group=Base Material
   parameterName=Base Material Mesh Data Masks Tex
   samplerType=SAMPLERTYPE_Masks
   sortPriority=4
   texture=/Game/Environments/Asteria/Foliage/Trees/KapokTree/Textures/KapokBark_masks
  }
  NamedRerouteDeclaration(None=@n45.R) {name=Base AO Mask, nodeColor=<rgba 1 0 0.6094 1>}
  NamedRerouteDeclaration(None=@n45.G) {name=Base Curv Mask, nodeColor=<rgba 1 0.2737 0 1>}
  NamedRerouteDeclaration(None=@n45.B) {name=Base Thickness Mask, nodeColor=<rgba 0 1 0.8096 1>}
  NamedRerouteDeclaration(None=@n44.Metallic) {name=Base Metallic, nodeColor=<rgba 0.2187 1 0 1>}
  NamedRerouteDeclaration(None=@n41) {name=Roughness Equalize Value, nodeColor=<rgba 0.6406 0 1 1>}
  n46 = TextureSampleParameter2D(UVs=@n36) {
   group=Detail
   parameterName=Detail Normal Tex
   samplerType=SAMPLERTYPE_Normal
   sortPriority=2
   texture=/Game/Environments/Asteria/Foliage/Trees/KapokTree/Textures/KapokBark_detail_normal
  }
  NamedRerouteDeclaration(None=Multiply(A=@n46.RGB, B=ScalarParameter() {defaultValue=0.25, group=Detail, parameterName=Detail Noirnal Strength})) {name=Detail Normal, nodeColor=<rgba 0.3384 0.267 1 1>}
  n47 = TextureSampleParameter2D(UVs=@n36) {
   group=Detail
   parameterName=Detail RCAO Tex
   sortPriority=3
   texture=/Game/Environments/Asteria/Foliage/Trees/KapokTree/Textures/KapokBark_detail_RCAO
  }
  NamedRerouteDeclaration(None=LinearInterpolate(A=@n47.R) {constAlpha=0.8}) {name=Detail Roughness, nodeColor=<rgba 0.0927 0.0959 0.0969 1>}
  NamedRerouteDeclaration(None=@n47.B) {name=Detail AO, nodeColor=<rgba 1 1 1 1>}
  n48 = VertexColor()
  NamedRerouteDeclaration(None=@n48.R) {name=Red VC Ground Falloff, nodeColor=<rgba 1 0.004 0 1>}
  NamedRerouteDeclaration(None=@n48.G) {name=Green VC Moss, nodeColor=<rgba 0 1 0 1>}
  NamedRerouteDeclaration(None=@n48.B) {name=Blue BC Vine, nodeColor=<rgba 0 0 1 1>}
  NamedRerouteDeclaration(None=@n44.Normal) {name=Base Normal, nodeColor=<rgba 0.3384 0.267 1 1>}
  NamedRerouteUsage()
 ]
}
